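# Physics asset for the Nosh Hunter female medium player body cosmetic, covering its dynamic scarf, belt and bullet bones.
# Box-shaped scarf bodies form short front, back and side chains with tight swing and twist limits, rooted on tiny MSTR bodies joined to spine_05.
# Two spine capsules give the upper torso volume.
# Belt and bullet bones have only tiny MSTR placeholder shapes, with bullets joined to right-side belt bodies by angularly free constraints.
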PhysicsAsset /BRCosmetics/Characters/Player/Female/Medium/Bodies/F_MED_NoshHunter/Meshes/Nosh_Hunter_Physics
bodies {
 dyn_belt_01_l_MSTR [{shapeType=Box, extentX=0.01, extentY=0.01, extentZ=0.01}]
 dyn_belt_01_r_MSTR [{shapeType=Box, extentX=0.01, extentY=0.01, extentZ=0.01}]
 dyn_belt_back_01_l_MSTR [{shapeType=Box, extentX=0.01, extentY=0.01, extentZ=0.01}]
 dyn_belt_back_01_r_MSTR [{shapeType=Box, extentX=0.01, extentY=0.01, extentZ=0.01}]
 dyn_belt_front_01_l_MSTR [{shapeType=Box, extentX=0.01, extentY=0.01, extentZ=0.01}]
 dyn_belt_front_01_r_MSTR [{shapeType=Box, extentX=0.01, extentY=0.01, extentZ=0.01}]
 dyn_bullet1_01_r_MSTR [{shapeType=Capsule, length=0.01, radius=0.01}]
 dyn_bullet2_01_r_MSTR [{shapeType=Capsule, length=0.01, radius=0.01}]
 dyn_bullet3_01_r_MSTR [{shapeType=Capsule, length=0.01, radius=0.01}]
 dyn_bullet4_01_r_MSTR [{shapeType=Capsule, length=0.01, radius=0.01}]
 dyn_bullet5_01_r_MSTR [{shapeType=Capsule, length=0.01, radius=0.01}]
 dyn_bullet6_01_r_MSTR [{shapeType=Capsule, length=0.01, radius=0.01}]
 dyn_bullet7_01_r_MSTR [{shapeType=Capsule, length=0.01, radius=0.01}]
 dyn_bullet8_01_r_MSTR [{shapeType=Capsule, length=0.01, radius=0.01}]
 dyn_scarf_back_01 [{shapeType=Box, extentX=4, extentY=12, extentZ=1.5, center=<2.0483 0 0>}]
 dyn_scarf_back_01_MSTR [{shapeType=Capsule, length=0.01, radius=0.01}]
 dyn_scarf_back_02 [{shapeType=Box, extentX=4, extentY=10.5, extentZ=1.5, center=<2.0483 0 0>}]
 dyn_scarf_front_01 [{shapeType=Box, extentX=4, extentY=12, extentZ=1.5, center=<2.0483 0 0>}]
 dyn_scarf_front_01_MSTR [{shapeType=Capsule, length=0.01, radius=0.01}]
 dyn_scarf_front_01_l [{shapeType=Box, extentX=4, extentY=8, extentZ=0.5, center=<2.0483 0 0>}]
 dyn_scarf_front_01_l_MSTR [{shapeType=Capsule, length=0.01, radius=0.01}]
 dyn_scarf_front_01_r [{shapeType=Box, extentX=4, extentY=8, extentZ=0.5, center=<-2.0483 0 0>, rotation=<rot 0 0 180>}]
 dyn_scarf_front_01_r_MSTR [{shapeType=Capsule, length=0.01, radius=0.01}]
 dyn_scarf_front_02 [{shapeType=Box, extentX=4, extentY=10.5, extentZ=1.5, center=<2.0483 0 0>}]
 spine_04 [{shapeType=Capsule, length=9.0219, radius=8.9291, center=<4.5816 1.8078 0>}]
 spine_05 [{shapeType=Capsule, length=5.7719, radius=8.9291, center=<2.0976 -0.3982 0>}]
}
constraints [
 {
  bone1Name=dyn_bullet1_01_r_MSTR
  bone2Name=dyn_belt_back_01_r_MSTR
  swing1LimitDegrees=45
  swing1Motion=Free
  swing2LimitDegrees=45
  swing2Motion=Free
  twistLimitDegrees=45
  twistMotion=Free
 }
 {
  bone1Name=dyn_bullet6_01_r_MSTR
  bone2Name=dyn_belt_front_01_r_MSTR
  swing1LimitDegrees=45
  swing1Motion=Free
  swing2LimitDegrees=45
  swing2Motion=Free
  twistLimitDegrees=45
  twistMotion=Free
 }
 {
  bone1Name=dyn_bullet7_01_r_MSTR
  bone2Name=dyn_belt_front_01_r_MSTR
  swing1LimitDegrees=45
  swing1Motion=Free
  swing2LimitDegrees=45
  swing2Motion=Free
  twistLimitDegrees=45
  twistMotion=Free
 }
 {
  bone1Name=dyn_bullet8_01_r_MSTR
  bone2Name=dyn_belt_front_01_r_MSTR
  swing1LimitDegrees=45
  swing1Motion=Free
  swing2LimitDegrees=45
  swing2Motion=Free
  twistLimitDegrees=45
  twistMotion=Free
 }
 {
  bone1Name=dyn_bullet2_01_r_MSTR
  bone2Name=dyn_belt_back_01_r_MSTR
  swing1LimitDegrees=45
  swing1Motion=Free
  swing2LimitDegrees=45
  swing2Motion=Free
  twistLimitDegrees=45
  twistMotion=Free
 }
 {
  bone1Name=dyn_bullet5_01_r_MSTR
  bone2Name=dyn_belt_01_r_MSTR
  swing1LimitDegrees=45
  swing1Motion=Free
  swing2LimitDegrees=45
  swing2Motion=Free
  twistLimitDegrees=45
  twistMotion=Free
 }
 {
  bone1Name=dyn_bullet4_01_r_MSTR
  bone2Name=dyn_belt_01_r_MSTR
  swing1LimitDegrees=45
  swing1Motion=Free
  swing2LimitDegrees=45
  swing2Motion=Free
  twistLimitDegrees=45
  twistMotion=Free
 }
 {
  bone1Name=dyn_bullet3_01_r_MSTR
  bone2Name=dyn_belt_01_r_MSTR
  swing1LimitDegrees=45
  swing1Motion=Free
  swing2LimitDegrees=45
  swing2Motion=Free
  twistLimitDegrees=45
  twistMotion=Free
 }
 {bone1Name=dyn_scarf_front_01_l, bone2Name=dyn_scarf_front_01_l_MSTR, swing1LimitDegrees=15, swing2LimitDegrees=25, twistLimitDegrees=5}
 {bone1Name=dyn_scarf_front_01, bone2Name=dyn_scarf_front_01_MSTR, swing1LimitDegrees=15, swing2LimitDegrees=25, twistLimitDegrees=5}
 {bone1Name=dyn_scarf_front_02, bone2Name=dyn_scarf_front_01, swing1LimitDegrees=15, swing2LimitDegrees=25, twistLimitDegrees=5}
 {bone1Name=dyn_scarf_back_01, bone2Name=dyn_scarf_back_01_MSTR, swing1LimitDegrees=15, swing2LimitDegrees=25, twistLimitDegrees=5}
 {bone1Name=dyn_scarf_back_02, bone2Name=dyn_scarf_back_01, swing1LimitDegrees=15, swing2LimitDegrees=25, twistLimitDegrees=5}
 {
  bone1Name=dyn_scarf_front_01_l_MSTR
  bone2Name=spine_05
  swing1LimitDegrees=45
  swing1Motion=Free
  swing2LimitDegrees=45
  swing2Motion=Free
  twistLimitDegrees=45
  twistMotion=Free
 }
 {
  bone1Name=dyn_scarf_front_01_MSTR
  bone2Name=spine_05
  swing1LimitDegrees=45
  swing1Motion=Free
  swing2LimitDegrees=45
  swing2Motion=Free
  twistLimitDegrees=45
  twistMotion=Free
 }
 {
  bone1Name=dyn_scarf_front_01_r_MSTR
  bone2Name=spine_05
  swing1LimitDegrees=45
  swing1Motion=Free
  swing2LimitDegrees=45
  swing2Motion=Free
  twistLimitDegrees=45
  twistMotion=Free
 }
 {
  bone1Name=dyn_scarf_back_01_MSTR
  bone2Name=spine_05
  swing1LimitDegrees=45
  swing1Motion=Free
  swing2LimitDegrees=45
  swing2Motion=Free
  twistLimitDegrees=45
  twistMotion=Free
 }
 {
  bone1Name=spine_05
  bone2Name=spine_04
  swing1LimitDegrees=45
  swing1Motion=Free
  swing2LimitDegrees=45
  swing2Motion=Free
  twistLimitDegrees=45
  twistMotion=Free
 }
 {bone1Name=dyn_scarf_front_01_r, bone2Name=dyn_scarf_front_01_r_MSTR, swing1LimitDegrees=15, swing2LimitDegrees=25, twistLimitDegrees=5}
]
