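# Physics asset for the left belt of the female medium NoshHunter outfit, rigging its front, back and canteen pieces for secondary motion.
# Belt capsules hang from tiny box master bodies with tight angular limits, and front and back pieces cross-link to the two-segment canteen chain.
# A left thigh capsule has no constraints, giving a leg collision shape.
# Three tiny right-side bullet bodies attach to the right front master with free rotation.
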PhysicsAsset /BRCosmetics/Characters/Player/Female/Medium/Bodies/F_MED_NoshHunter/Meshes/Nosh_Hunter_belt_l_Physics
bodies {
 dyn_belt_01_l_MSTR [{shapeType=Box, extentX=0.01, extentY=0.01, extentZ=0.01}]
 dyn_belt_back_01_l [{shapeType=Capsule, length=7.5598, radius=2.8089, center=<7.0728 -0.0005 0>, rotation=<rot -90 6.7213 -6.7212>}]
 dyn_belt_back_01_l_MSTR [{shapeType=Box, extentX=0.01, extentY=0.01, extentZ=0.01}]
 dyn_belt_canteen_01_l [{shapeType=Capsule, length=7.5598, radius=2.8089, center=<5.8994 -0.0005 0>, rotation=<rot -90 -5.0512 5.0513>}]
 dyn_belt_canteen_02_l [{shapeType=Capsule, length=8.9117, radius=4.6969, center=<1.7188 0 3.0979>, rotation=<rot 90 5.0515 5.0521>}]
 dyn_belt_front_01_l [{shapeType=Capsule, length=7.5598, radius=2.8089, center=<5.8994 -0.0005 0>, rotation=<rot -90 -5.0512 5.0513>}]
 dyn_belt_front_01_l_MSTR [{shapeType=Box, extentX=0.01, extentY=0.01, extentZ=0.01}]
 dyn_belt_front_01_r_MSTR [{shapeType=Box, extentX=0.01, extentY=0.01, extentZ=0.01}]
 dyn_bullet6_01_r_MSTR [{shapeType=Capsule, length=0.01, radius=0.01}]
 dyn_bullet7_01_r_MSTR [{shapeType=Capsule, length=0.01, radius=0.01}]
 dyn_bullet8_01_r_MSTR [{shapeType=Capsule, length=0.01, radius=0.01}]
 thigh_l [{shapeType=Capsule, length=23.4861, radius=5.4482, center=<-13.3444 1.4828 0.3289>, rotation=<rot -90 -17.354 12.3539>}]
}
constraints [
 {bone1Name=dyn_belt_canteen_01_l, bone2Name=dyn_belt_01_l_MSTR, swing1LimitDegrees=5, swing2LimitDegrees=15, twistLimitDegrees=5}
 {bone1Name=dyn_belt_canteen_02_l, bone2Name=dyn_belt_canteen_01_l, swing1LimitDegrees=10, swing2LimitDegrees=10, twistLimitDegrees=5}
 {bone1Name=dyn_belt_back_01_l, bone2Name=dyn_belt_back_01_l_MSTR, swing1LimitDegrees=5, swing2LimitDegrees=25, twistLimitDegrees=5}
 {bone1Name=dyn_belt_front_01_l, bone2Name=dyn_belt_front_01_l_MSTR, swing1LimitDegrees=5, swing2LimitDegrees=20, twistLimitDegrees=5}
 {
  bone1Name=dyn_bullet6_01_r_MSTR
  bone2Name=dyn_belt_front_01_r_MSTR
  swing1LimitDegrees=45
  swing1Motion=Free
  swing2LimitDegrees=45
  swing2Motion=Free
  twistLimitDegrees=45
  twistMotion=Free
 }
 {
  bone1Name=dyn_bullet7_01_r_MSTR
  bone2Name=dyn_belt_front_01_r_MSTR
  swing1LimitDegrees=45
  swing1Motion=Free
  swing2LimitDegrees=45
  swing2Motion=Free
  twistLimitDegrees=45
  twistMotion=Free
 }
 {
  bone1Name=dyn_bullet8_01_r_MSTR
  bone2Name=dyn_belt_front_01_r_MSTR
  swing1LimitDegrees=45
  swing1Motion=Free
  swing2LimitDegrees=45
  swing2Motion=Free
  twistLimitDegrees=45
  twistMotion=Free
 }
 {bone1Name=dyn_belt_back_01_l, bone2Name=dyn_belt_canteen_01_l, swing1LimitDegrees=10, swing2LimitDegrees=10, twistLimitDegrees=45}
 {bone1Name=dyn_belt_front_01_l, bone2Name=dyn_belt_canteen_01_l, swing1LimitDegrees=10, swing2LimitDegrees=10, twistLimitDegrees=10}
]
