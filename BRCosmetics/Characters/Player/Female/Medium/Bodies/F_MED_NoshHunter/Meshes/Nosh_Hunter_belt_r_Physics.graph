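# Physics asset for the right belt of the NoshHunter female body cosmetic, defining jiggle for the belt and its eight bullets.
# Back, canteen and front belt capsules have limited swing and twist against tiny MSTR anchor boxes.
# Each bullet capsule is tightly limited to a tiny MSTR capsule, which joins a belt segment with free rotation.
# Extra limited constraints tie some bullets to the belt or neighboring bullets.
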PhysicsAsset /BRCosmetics/Characters/Player/Female/Medium/Bodies/F_MED_NoshHunter/Meshes/Nosh_Hunter_belt_r_Physics
bodies {
 dyn_belt_01_l_MSTR [{shapeType=Box, extentX=0.01, extentY=0.01, extentZ=0.01}]
 dyn_belt_01_r_MSTR [{shapeType=Box, extentX=0.01, extentY=0.01, extentZ=0.01}]
 dyn_belt_back_01_l_MSTR [{shapeType=Box, extentX=0.01, extentY=0.01, extentZ=0.01}]
 dyn_belt_back_01_r [{shapeType=Capsule, length=7.5285, radius=2.5594, center=<-5.8994 0.0005 0>, rotation=<rot -90 10.025 169.9751>}]
 dyn_belt_back_01_r_MSTR [{shapeType=Box, extentX=0.01, extentY=0.01, extentZ=0.01}]
 dyn_belt_canteen_01_r [{shapeType=Capsule, length=7.5285, radius=2.5594, center=<-5.8994 0.0005 0>, rotation=<rot -90 10.025 169.9751>}]
 dyn_belt_front_01_l_MSTR [{shapeType=Box, extentX=0.01, extentY=0.01, extentZ=0.01}]
 dyn_belt_front_01_r [{shapeType=Capsule, length=7.5285, radius=2.5594, center=<-5.8994 0.0005 0>, rotation=<rot -90 10.025 169.9751>}]
 dyn_belt_front_01_r_MSTR [{shapeType=Box, extentX=0.01, extentY=0.01, extentZ=0.01}]
 dyn_bullet1_01_r [{shapeType=Capsule, length=2.5305, radius=2.3626, center=<-0.5933 -0.0002 0.6997>, rotation=<rot -84.9998 -0.0001 0>}]
 dyn_bullet1_01_r_MSTR [{shapeType=Capsule, length=0.01, radius=0.01}]
 dyn_bullet2_01_r [{shapeType=Capsule, length=2.5305, radius=2.3626, center=<-0.5933 -0.0002 0.6997>, rotation=<rot -84.9998 -0.0001 0>}]
 dyn_bullet2_01_r_MSTR [{shapeType=Capsule, length=0.01, radius=0.01}]
 dyn_bullet3_01_r [{shapeType=Capsule, length=2.5305, radius=2.3626, center=<-0.5933 -0.0002 0.6997>, rotation=<rot -84.9998 -0.0001 0>}]
 dyn_bullet3_01_r_MSTR [{shapeType=Capsule, length=0.01, radius=0.01}]
 dyn_bullet4_01_r [{shapeType=Capsule, length=2.5305, radius=2.3626, center=<-0.5933 -0.0002 0.6997>, rotation=<rot -84.9998 -0.0001 0>}]
 dyn_bullet4_01_r_MSTR [{shapeType=Capsule, length=0.01, radius=0.01}]
 dyn_bullet5_01_r [{shapeType=Capsule, length=2.5305, radius=2.3626, center=<-0.5933 -0.0002 0.6997>, rotation=<rot -84.9998 -0.0001 0>}]
 dyn_bullet5_01_r_MSTR [{shapeType=Capsule, length=0.01, radius=0.01}]
 dyn_bullet6_01_r [{shapeType=Capsule, length=2.5305, radius=2.3626, center=<-0.5933 -0.0002 0.6997>, rotation=<rot -84.9998 -0.0001 0>}]
 dyn_bullet6_01_r_MSTR [{shapeType=Capsule, length=0.01, radius=0.01}]
 dyn_bullet7_01_r [{shapeType=Capsule, length=2.5305, radius=2.3626, center=<-0.5933 -0.0002 0.6997>, rotation=<rot -84.9998 -0.0001 0>}]
 dyn_bullet7_01_r_MSTR [{shapeType=Capsule, length=0.01, radius=0.01}]
 dyn_bullet8_01_r [{shapeType=Capsule, length=2.5305, radius=2.3626, center=<-0.5933 -0.0002 0.6997>, rotation=<rot -84.9998 -0.0001 0>}]
 dyn_bullet8_01_r_MSTR [{shapeType=Capsule, length=0.01, radius=0.01}]
}
constraints [
 {bone1Name=dyn_belt_back_01_r, bone2Name=dyn_belt_back_01_r_MSTR, swing1LimitDegrees=5, swing2LimitDegrees=20, twistLimitDegrees=5}
 {bone1Name=dyn_belt_canteen_01_r, bone2Name=dyn_belt_01_r_MSTR, swing1LimitDegrees=5, swing2LimitDegrees=20, twistLimitDegrees=5}
 {bone1Name=dyn_belt_front_01_r, bone2Name=dyn_belt_front_01_r_MSTR, swing1LimitDegrees=5, swing2LimitDegrees=20, twistLimitDegrees=5}
 {
  bone1Name=dyn_bullet1_01_r_MSTR
  bone2Name=dyn_belt_back_01_r
  swing1LimitDegrees=45
  swing1Motion=Free
  swing2LimitDegrees=45
  swing2Motion=Free
  twistLimitDegrees=45
  twistMotion=Free
 }
 {
  bone1Name=dyn_bullet6_01_r_MSTR
  bone2Name=dyn_belt_front_01_r
  swing1LimitDegrees=45
  swing1Motion=Free
  swing2LimitDegrees=45
  swing2Motion=Free
  twistLimitDegrees=45
  twistMotion=Free
 }
 {
  bone1Name=dyn_bullet7_01_r_MSTR
  bone2Name=dyn_belt_front_01_r
  swing1LimitDegrees=45
  swing1Motion=Free
  swing2LimitDegrees=45
  swing2Motion=Free
  twistLimitDegrees=45
  twistMotion=Free
 }
 {
  bone1Name=dyn_bullet8_01_r_MSTR
  bone2Name=dyn_belt_front_01_r
  swing1LimitDegrees=45
  swing1Motion=Free
  swing2LimitDegrees=45
  swing2Motion=Free
  twistLimitDegrees=45
  twistMotion=Free
 }
 {
  bone1Name=dyn_bullet2_01_r_MSTR
  bone2Name=dyn_belt_back_01_r
  swing1LimitDegrees=45
  swing1Motion=Free
  swing2LimitDegrees=45
  swing2Motion=Free
  twistLimitDegrees=45
  twistMotion=Free
 }
 {
  bone1Name=dyn_bullet5_01_r_MSTR
  bone2Name=dyn_belt_canteen_01_r
  swing1LimitDegrees=45
  swing1Motion=Free
  swing2LimitDegrees=45
  swing2Motion=Free
  twistLimitDegrees=45
  twistMotion=Free
 }
 {
  bone1Name=dyn_bullet4_01_r_MSTR
  bone2Name=dyn_belt_canteen_01_r
  swing1LimitDegrees=45
  swing1Motion=Free
  swing2LimitDegrees=45
  swing2Motion=Free
  twistLimitDegrees=45
  twistMotion=Free
 }
 {
  bone1Name=dyn_bullet3_01_r_MSTR
  bone2Name=dyn_belt_canteen_01_r
  swing1LimitDegrees=45
  swing1Motion=Free
  swing2LimitDegrees=45
  swing2Motion=Free
  twistLimitDegrees=45
  twistMotion=Free
 }
 {bone1Name=dyn_bullet8_01_r, bone2Name=dyn_bullet8_01_r_MSTR, swing1LimitDegrees=5, swing2LimitDegrees=5, twistLimitDegrees=5}
 {bone1Name=dyn_bullet7_01_r, bone2Name=dyn_bullet7_01_r_MSTR, swing1LimitDegrees=5, swing2LimitDegrees=5, twistLimitDegrees=5}
 {bone1Name=dyn_bullet6_01_r, bone2Name=dyn_bullet6_01_r_MSTR, swing1LimitDegrees=5, swing2LimitDegrees=5, twistLimitDegrees=5}
 {bone1Name=dyn_bullet1_01_r, bone2Name=dyn_bullet1_01_r_MSTR, swing1LimitDegrees=5, swing2LimitDegrees=5, twistLimitDegrees=5}
 {bone1Name=dyn_bullet2_01_r, bone2Name=dyn_bullet2_01_r_MSTR, swing1LimitDegrees=10, swing2LimitDegrees=5, twistLimitDegrees=5}
 {bone1Name=dyn_bullet5_01_r, bone2Name=dyn_bullet5_01_r_MSTR, swing1LimitDegrees=5, swing2LimitDegrees=5, twistLimitDegrees=5}
 {bone1Name=dyn_bullet4_01_r, bone2Name=dyn_bullet4_01_r_MSTR, swing1LimitDegrees=10, swing2LimitDegrees=5, twistLimitDegrees=5}
 {bone1Name=dyn_bullet3_01_r, bone2Name=dyn_bullet3_01_r_MSTR, swing1LimitDegrees=5, swing2LimitDegrees=5, twistLimitDegrees=5}
 {bone1Name=dyn_bullet1_01_r, bone2Name=dyn_belt_back_01_r, swing1LimitDegrees=0, swing2LimitDegrees=10, twistLimitDegrees=5}
 {bone1Name=dyn_bullet2_01_r, bone2Name=dyn_belt_back_01_r, swing1LimitDegrees=10, swing2LimitDegrees=10, twistLimitDegrees=5}
 {bone1Name=dyn_bullet3_01_r, bone2Name=dyn_bullet2_01_r, swing1LimitDegrees=5, swing2LimitDegrees=5, twistLimitDegrees=5, twistMotion=Locked}
 {bone1Name=dyn_bullet2_01_r, bone2Name=dyn_bullet1_01_r, swing1LimitDegrees=5, swing2LimitDegrees=5, twistLimitDegrees=5, twistMotion=Locked}
 {bone1Name=dyn_bullet4_01_r, bone2Name=dyn_belt_canteen_01_r, swing1LimitDegrees=10, swing2LimitDegrees=10, twistLimitDegrees=10}
 {bone1Name=dyn_bullet4_01_r, bone2Name=dyn_bullet5_01_r, swing1LimitDegrees=5, swing2LimitDegrees=5, twistLimitDegrees=5}
]
