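# Post-process style Anim Blueprint for the Nosh Hunter female body, adding secondary motion and belt corrections over a pose copied from the attached parent mesh.
# Additive bone offsets, blended by crouch, skydive and thigh-swing alphas, surround three RigidBody simulations for body and left/right belts, plus rotation constraints.
# The update event derives alphas from ref-pose bone deltas and swaps body-sim gravity for scaled skydiving force.
# Shoulder-inflate nodes are unconnected to the output.
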AnimBlueprint /BRCosmetics/Characters/Player/Female/Medium/Bodies/F_MED_NoshHunter/Meshes/Nosh_Hunter_AnimBP
target_skeleton=&/Nosh_Hunter_Skeleton
graphs [
 AnimGraph {nodes [
   n1 = "Transform (Modify) Bone - Bone: upperarm_twist_01_r"(Alpha=Get shoulderInflate_r() -> (shoulderInflate_r:double), Translation="0,0,0", Rotation="0,0,0", Scale="1.1,1.1,1.1") {scaleMode=BMM_Additive, alphaBoolBlend {blendOption=Linear}, boneToModify {boneName=upperarm_twist_01_r}, scale=<1.1 1.1 1.1>}
   "Transform (Modify) Bone - Bone: upperarm_twist_01_l"(ComponentPose=@n1, Alpha=Get shoulderInflate_l() -> (shoulderInflate_l:double), Translation="0,0,0", Rotation="0,0,0", Scale="1.1,1.1,1.1") {scaleMode=BMM_Additive, alphaBoolBlend {blendOption=Linear}, boneToModify {boneName=upperarm_twist_01_l}, scale=<1.1 1.1 1.1>}
   n2 = Get isCrouchingAlpha() -> (isCrouchingAlpha:double)
   "Transform (Modify) Bone - Bone: dyn_belt_back_02_l"(Alpha=@n2, Translation="0,0,0", Rotation="-25,0,0", Scale="1,1,1") {
    rotationMode=BMM_Additive
    rotationSpace=BCS_BoneSpace
    scaleMode=BMM_Additive
    scaleSpace=BCS_BoneSpace
    translationMode=BMM_Additive
    translationSpace=BCS_BoneSpace
    alphaBoolBlend {blendOption=Linear}
    boneToModify {boneName=dyn_belt_back_02_l}
    rotation {pitch=-25}
   }
   n3 = Get thighSwingDeltaWalk_l() -> (thighSwingDeltaWalk_l:double)
   "Transform (Modify) Bone - Bone: dyn_belt_01_l_MSTR"(Alpha=@n3, Translation="0,0,0", Rotation="15,0,0", Scale="1,1,1") {
    rotationMode=BMM_Additive
    rotationSpace=BCS_BoneSpace
    translationMode=BMM_Additive
    translationSpace=BCS_BoneSpace
    alphaBoolBlend {blendOption=Linear}
    boneToModify {boneName=dyn_belt_01_l_MSTR}
    rotation {pitch=15}
   }
   n4 = Get isCrouchingAlpha()
   n5 = "Transform (Modify) Bone - Bone: dyn_belt_front_01_l_MSTR"(
    ComponentPose=Local To Component(LocalPose=Copy Pose From Mesh() {bUseAttachedParent=true})
    Alpha=@n4
    Translation="0,0,0"
    Rotation="30,0,0"
    Scale="1,1,1"
   ) {
    rotationMode=BMM_Additive
    rotationSpace=BCS_BoneSpace
    translationMode=BMM_Additive
    translationSpace=BCS_BoneSpace
    alphaBoolBlend {blendOption=Linear}
    boneToModify {boneName=dyn_belt_front_01_l_MSTR}
    rotation {pitch=30}
   }
   n6 = "Transform (Modify) Bone - Bone: dyn_belt_01_l_MSTR"(ComponentPose=@n5, Alpha=@n4, Translation="0,0,0", Rotation="0,-40,0", Scale="1,1,1") {
    rotationMode=BMM_Additive
    rotationSpace=BCS_BoneSpace
    translationMode=BMM_Additive
    translationSpace=BCS_BoneSpace
    alphaBoolBlend {blendOption=Linear}
    boneToModify {boneName=dyn_belt_01_l_MSTR}
    rotation {yaw=-40}
   }
   n7 = "Transform (Modify) Bone - Bone: dyn_belt_back_01_l_MSTR"(ComponentPose=@n6, Alpha=@n4, Translation="0,0,0", Rotation="0,-10,0", Scale="1,1,1") {
    rotationMode=BMM_Additive
    rotationSpace=BCS_BoneSpace
    translationMode=BMM_Additive
    translationSpace=BCS_BoneSpace
    alphaBoolBlend {blendOption=Linear}
    boneToModify {boneName=dyn_belt_back_01_l_MSTR}
    rotation {yaw=-10}
   }
   n8 = "Transform (Modify) Bone - Bone: dyn_belt_front_01_l_MSTR"(ComponentPose=@n7, Alpha=@n3, Translation="1,0,0", Rotation="15,0,0", Scale="1,1,1") {
    rotationMode=BMM_Additive
    rotationSpace=BCS_BoneSpace
    translationMode=BMM_Additive
    translationSpace=BCS_BoneSpace
    alphaBoolBlend {blendOption=Linear}
    boneToModify {boneName=dyn_belt_front_01_l_MSTR}
    rotation {pitch=15}
    translation {x=1}
   }
   n9 = RigidBody!(ComponentPose=@n8, Alpha, OverrideWorldGravity:Vector=Get windVector() -> (windVector:Vector), ExternalForce) -> (Pose) {
    bClampLinearTranslationLimitToRefPose=true
    bOverrideWorldGravity=true
    overridePhysicsAsset=&/Nosh_Hunter_Physics
    simulationSpace=BaseBoneSpace
    alphaBoolBlend {blendOption=Linear}
    baseBoneRef {boneName=spine_01}
    componentAppliedLinearAccClamp=<3000 3000 3000>
    componentLinearAccScale=<0.5 0.5 0.5>
    componentLinearVelScale=<0.7 0.7 0.7>
    overrideWorldGravity {z=-980}
   }
   n10 = RigidBody(ComponentPose=@n9, Alpha=0.85, OverrideWorldGravity="0,0,-980") {
    bClampLinearTranslationLimitToRefPose=true
    bOverrideWorldGravity=true
    overridePhysicsAsset=&/Nosh_Hunter_belt_l_Physics
    simulationSpace=BaseBoneSpace
    alphaBoolBlend {blendOption=Linear}
    baseBoneRef {boneName=pelvis}
    componentAppliedLinearAccClamp=<1000 1000 1000>
    componentLinearAccScale=<0.5 0.2 0.2>
    componentLinearVelScale=<0.5 0.2 0.2>
    overrideWorldGravity {z=-980}
   }
   n11 = RigidBody!(ComponentPose=@n10, Alpha, ExternalForce) -> (Pose) {
    bClampLinearTranslationLimitToRefPose=true
    bOverrideWorldGravity=true
    overridePhysicsAsset=&/Nosh_Hunter_belt_r_Physics
    simulationSpace=BaseBoneSpace
    alphaBoolBlend {blendOption=Linear}
    baseBoneRef {boneName=pelvis}
    componentAppliedLinearAccClamp=<1000 1000 1000>
    componentLinearAccScale=<0.5 0.2 0.2>
    componentLinearVelScale=<0.5 0.2 0.2>
    overrideWorldGravity {z=-980}
   }
   n12 = "Transform (Modify) Bone - Bone: thigh_back_l_01_DSP"(
    ComponentPose=@n11
    Alpha=Get thighSwingDeltaCrouch_l() -> (thighSwingDeltaCrouch_l:double)
    Translation="0,0,0"
    Rotation="0.000,0.000,-10"
    Scale="1,1,1"
   ) {
    rotationMode=BMM_Additive
    scaleMode=BMM_Additive
    translationMode=BMM_Additive
    alphaBoolBlend {blendOption=Linear}
    boneToModify {boneName=thigh_back_l_01_DSP}
    rotation {roll=-10}
   }
   n13 = "Transform (Modify) Bone - Bone: thigh_back_r_01_DSP"(
    ComponentPose=@n12
    Alpha=Get thighSwingDeltaCrouch_r() -> (thighSwingDeltaCrouch_r:double)
    Translation="0,-1,-2"
    Rotation="0.000,0.000,-2.5"
    Scale="1,1,1"
   ) {
    rotationMode=BMM_Additive
    scaleMode=BMM_Additive
    translationMode=BMM_Additive
    alphaBoolBlend {blendOption=Linear}
    boneToModify {boneName=thigh_back_r_01_DSP}
    rotation {roll=-2.5}
    translation {y=-1, z=-2}
   }
   n14 = "Transform (Modify) Bone - Bone: dyn_belt_back_01_l_MSTR"(ComponentPose=@n13, Alpha=@n2, Translation="0,0,0", Rotation="0,0,0", Scale="0.9,1,1") {
    rotationMode=BMM_Additive
    rotationSpace=BCS_BoneSpace
    scaleMode=BMM_Additive
    scaleSpace=BCS_BoneSpace
    translationMode=BMM_Additive
    translationSpace=BCS_BoneSpace
    alphaBoolBlend {blendOption=Linear}
    boneToModify {boneName=dyn_belt_back_01_l_MSTR}
    scale {x=0.9}
   }
   n15 = "Constraint - dyn_belt_middle_01_l_MSTR (R)"(ComponentPose=@n14, Alpha=0.75, ConstraintWeights_0=1) {
    alphaBoolBlend {blendOption=Linear}
    boneToModify {boneName=dyn_belt_middle_01_l_MSTR}
    constraintSetup [{transformType=Rotation, targetBone {boneName=dyn_belt_back_01_l}}]
    constraintWeights [1]
   }
   n16 = "Constraint - dyn_belt_back_01_r (R)"(ComponentPose=@n15, Alpha=0.75, ConstraintWeights_0=1) {
    alphaBoolBlend {blendOption=Linear}
    boneToModify {boneName=dyn_belt_back_01_r}
    constraintSetup [{transformType=Rotation, targetBone {boneName=dyn_belt_middle_01_l}}]
    constraintWeights [1]
   }
   n17 = "Transform (Modify) Bone - Bone: dyn_belt_middle_01_l"(ComponentPose=@n16, Alpha=1, Translation="1.509,0.000,-0.000", Rotation="0,0,0", Scale="1,1,1") {
    translationMode=BMM_Additive
    translationSpace=BCS_BoneSpace
    alphaBoolBlend {blendOption=Linear}
    boneToModify {boneName=dyn_belt_middle_01_l}
    translation=<1.5088 0.0002 0>
   }
   n18 = Get belt_fix() -> (belt_fix:double)
   n19 = "Transform (Modify) Bone - Bone: dyn_belt_back_02_l"(ComponentPose=@n17, Alpha=@n18, Translation="3,1.5,-1", Rotation="-15,0,0", Scale="1,1,1") {
    rotationMode=BMM_Additive
    rotationSpace=BCS_BoneSpace
    translationMode=BMM_Additive
    translationSpace=BCS_BoneSpace
    alphaBoolBlend {blendOption=Linear}
    boneToModify {boneName=dyn_belt_back_02_l}
    rotation {pitch=-15}
    translation=<3 1.5 -1>
   }
   n20 = "Transform (Modify) Bone - Bone: dyn_belt_middle_02_l"(ComponentPose=@n19, Alpha=@n18, Translation="-0.000,-1,0.000", Rotation="0,0,0", Scale="1,1,1") {
    rotationMode=BMM_Additive
    rotationSpace=BCS_BoneSpace
    translationMode=BMM_Additive
    translationSpace=BCS_BoneSpace
    alphaBoolBlend {blendOption=Linear}
    boneToModify {boneName=dyn_belt_middle_02_l}
    translation {y=-1}
   }
   n21 = "Transform (Modify) Bone - Bone: dyn_belt_canteen_02_l"(ComponentPose=@n20, Alpha=@n18, Translation="0,0,2", Rotation="0,0,0", Scale="1,1,1") {
    rotationMode=BMM_Additive
    rotationSpace=BCS_BoneSpace
    translationMode=BMM_Additive
    translationSpace=BCS_BoneSpace
    alphaBoolBlend {blendOption=Linear}
    boneToModify {boneName=dyn_belt_canteen_02_l}
    translation {z=2}
   }
   n22 = "Transform (Modify) Bone - Bone: dyn_belt_back_02_r"!(ComponentPose=@n21, bAlphaBoolEnabled:bool=True, Translation="-6,5,2.000", Rotation="-0.000,20.785,-0.000", Scale="1,1,1") -> (Pose) {
    alphaInputType=Bool
    rotationMode=BMM_Additive
    rotationSpace=BCS_BoneSpace
    translationMode=BMM_Additive
    translationSpace=BCS_BoneSpace
    alphaBoolBlend {blendOption=Linear}
    boneToModify {boneName=dyn_belt_back_02_r}
    rotation=<rot 0 20.785 -0.0001>
    translation=<-6 5 2>
   }
   Output Pose(Result=Component To Local(ComponentPose=@n22)) {layerGroup=None}
  ]}
 EventGraph {nodes [
   ResetDynamics(InTeleportType=TeleportPhysics)
   > ResetDynamics(InTeleportType=ResetPhysics)
   ResetDynamics(InTeleportType=TeleportPhysics)
   > ResetDynamics(InTeleportType=ResetPhysics)
   Delay(Duration=0.125)
   > ResetDynamics(InTeleportType=TeleportPhysics)
   > ResetDynamics(InTeleportType=ResetPhysics)
   Event BlueprintUpdateAnimation()
   > n1 = Sequence!(execute) -> (then_0, then_1, then_2:Exec)
   n2 = Break Fort Anim Input Skydiving External Force(FortAnimInput_SkydivingExternalForce=Get SkydivingExternalForceInput()) -> (
    bUseSkydivingVectorForce=~
    SkydivingVectorForce=~
    HeadToPelvisDirection=~
    FloatingMultiplier=~
    FloatingAdditive=~
    DivingMultiplier=~
    DivingAdditive=~
    ParachutingMultiplier=~
    ParachutingAdditive=~
   )
   > Set windVector(windVector:Vector=SelectVector(A=vector * float(A=@n2.SkydivingVectorForce, B=0.15), B="0, 0,-980", bPickA=Get bIsSkyDiving())) -> (Output_Get:Vector)
   n3 = GetOwnerSkeletalMeshComponent()
   n4 = BreakTransform(InTransform=GetDeltaTransformFromRefPose(self=@n3, BoneName=head, BaseName=spine_05))
   n5 = Get bIsSkyDiving()
   Set headSwingAlpha(
    execute=@n1.then_1
    headSwingAlpha:double=MapRangeClamped(Value=BreakRotator(InRot=@n4.Rotation).Yaw, InRangeA=-45, InRangeB=0, OutRangeA=SelectFloat(A=1, B=0.75, bPickA=@n5), OutRangeB=0)
   ) -> (Output_Get:double)
   n6 = GetDeltaTransformFromRefPose(self=@n3, BoneName=upperarm_r, BaseName=clavicle_r)
   n7 = "Absolute (Float)"(A=BreakRotator(InRot=BreakTransform(InTransform=@n6).Rotation).Yaw)
   > Set armSwingDelta_r(armSwingDelta_r:double=MapRangeClamped(Value=@n7, InRangeA=5, InRangeB=70)) -> (Output_Get:double)
   n8 = GetDeltaTransformFromRefPose(self=@n3, BoneName=upperarm_l, BaseName=clavicle_l)
   n9 = "Absolute (Float)"(A=BreakRotator(InRot=BreakTransform(InTransform=@n8).Rotation).Yaw)
   > Set armSwingDelta_l(armSwingDelta_l:double=MapRangeClamped(Value=@n9, InRangeA=5, InRangeB=70)) -> (Output_Get:double)
   n10 = Get bIsSkyDiving()
   > Set shoulderInflate_r(shoulderInflate_r:double=SelectFloat(A=0.5, B=MapRangeClamped(Value=@n7, InRangeA=90, InRangeB=130, OutRangeB=0.75), bPickA=@n10)) -> (Output_Get:double)
   > Set shoulderInflate_l(shoulderInflate_l:double=SelectFloat(A=0.5, B=MapRangeClamped(Value=@n9, InRangeA=10, InRangeB=80), bPickA=@n10)) -> (Output_Get:double)
   n11 = GetDeltaTransformFromRefPose(self=@n3, BoneName=thigh_l, BaseName=pelvis)
   n12 = "Absolute (Float)"(A=BreakRotator(InRot=BreakTransform(InTransform=@n11).Rotation).Yaw)
   > Set thighSwingDeltaCrouch_l(thighSwingDeltaCrouch_l:double=MapRangeClamped(Value=@n12, InRangeA=50, InRangeB=110)) -> (Output_Get:double)
   > Set thighSwingDeltaWalk_l(thighSwingDeltaWalk_l:double=SelectFloat(A=0, B=MapRangeClamped(Value=@n12, InRangeA=20, InRangeB=50, OutRangeA=1, OutRangeB=0), bPickA=Get bIsCrouching())) -> (Output_Get:double)
   n13 = BreakRotator(InRot=BreakTransform(InTransform=GetDeltaTransformFromRefPose(self=@n3, BoneName=thigh_r, BaseName=pelvis)).Rotation)
   > Set thighSwingDeltaCrouch_r(thighSwingDeltaCrouch_r:double=MapRangeClamped(Value="Absolute (Float)"(A=@n13.Yaw), InRangeA=50, InRangeB=110)) -> (Output_Get:double)
   n14 = Get bIsSkyDiving()
   n15 = Get bIsCrouching()
   Set belt_fix(execute=@n1.then_2, belt_fix:double=SelectFloat(A=1, B=0, bPickA=OR Boolean!(A=@n15, B=@n14, C:bool) -> (ReturnValue))) -> (Output_Get:double)
   Event BlueprintInitializeAnimation()
   > Set isCrouchingAlpha(isCrouchingAlpha:double=0) -> (Output_Get:double)
   > Set isNotCrouchingAlpha(isNotCrouchingAlpha:double=1) -> (Output_Get:double)
   Event OnBeginCrouching()
   > Set isCrouchingAlpha(isCrouchingAlpha=1)
   > Set isNotCrouchingAlpha(isNotCrouchingAlpha=0)
   Event OnEndCrouching()
   > Set isCrouchingAlpha(isCrouchingAlpha=0)
   > Set isNotCrouchingAlpha(isNotCrouchingAlpha=1)
  ]}
]
variables [
 armSwingDelta_l {type=number}
 armSwingDelta_r {type=number}
 belt_fix {type=number}
 headSwingAlpha {type=number}
 isCrouchingAlpha {type=number}
 isNotCrouchingAlpha {type=number}
 shoulderInflate_l {type=number}
 shoulderInflate_r {type=number}
 thighSwingDeltaCrouch_l {type=number}
 thighSwingDeltaCrouch_r {type=number}
 thighSwingDeltaWalk_l {type=number}
 windVector {type=Vector}
]
