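# Physics asset for the female medium Nosh Hunter face accessory mesh, providing secondary motion for its dynamic hair bones.
# Top, front, side and long hair form box or capsule chains, each rooted in a tiny MSTR anchor body constrained to the head.
# Hair links carry per-chain swing and twist limits, while anchor and body links rotate freely.
# Head, neck, upper spine and upper arm capsules supply body shapes around the hair.
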PhysicsAsset /BRCosmetics/Characters/Player/Female/Medium/Bodies/F_MED_NoshHunter/Meshes/Parts/F_MED_Nosh_Hunter_FaceAcc_Physics
bodies {
 dyn_fronthair_01_l [{shapeType=Box, extentX=4, extentY=5, extentZ=1.1, center=<1.5998 0 0>}]
 dyn_fronthair_01_l_MSTR [{shapeType=Capsule, length=0.01, radius=0.01}]
 dyn_fronthair_01_r [{shapeType=Capsule, length=2.975, radius=1.1834, center=<-1.4458 0 0>, rotation=<rot 90 0 0>}]
 dyn_fronthair_01_r_MSTR [{shapeType=Capsule, length=0.01, radius=0.01}]
 dyn_fronthair_02_l [{shapeType=Box, extentX=4, extentY=5, extentZ=1.1, center=<1.5998 0 0>}]
 dyn_fronthair_02_r [{shapeType=Capsule, length=2.975, radius=1.1834, center=<-1.4458 0 0>, rotation=<rot 90 0 0>}]
 dyn_longhair_01 [{shapeType=Box, extentX=4, extentY=5, extentZ=4.6, center=<1.5998 0 0>}]
 dyn_longhair_01_MSTR [{shapeType=Capsule, length=0.01, radius=0.01}]
 dyn_longhair_01_l [{shapeType=Capsule, length=3.0253, radius=2.2355, center=<1.5998 0 0>, rotation=<rot 90 -0.0396 -0.0396>}]
 dyn_longhair_01_l_MSTR [{shapeType=Capsule, length=0.01, radius=0.01}]
 dyn_longhair_01_r [{shapeType=Capsule, length=3.4253, radius=2.597, center=<-2.8498 0.0138 -0.1966>, rotation=<rot 90 0 0>}]
 dyn_longhair_01_r_MSTR [{shapeType=Capsule, length=0.01, radius=0.01}]
 dyn_longhair_02 [{shapeType=Box, extentX=4, extentY=5, extentZ=4.6, center=<1.5998 0 0>}]
 dyn_longhair_02_l [{shapeType=Capsule, length=3.0253, radius=2.2355, center=<1.5998 0 0>, rotation=<rot 90 -0.0396 -0.0396>}]
 dyn_longhair_02_r [{shapeType=Capsule, length=1.4253, radius=2.597, center=<-2.8498 0.0138 -0.1966>, rotation=<rot 90 0 0>}]
 dyn_longhair_03 [{shapeType=Box, extentX=4, extentY=5, extentZ=4.6, center=<1.5998 0 0>}]
 dyn_longhair_03_l [{shapeType=Capsule, length=3.7753, radius=1.4855, center=<1.5998 0 0>, rotation=<rot 90 -0.0396 -0.0396>}]
 dyn_longhair_03_r [{shapeType=Capsule, length=2.6753, radius=2.097, center=<-2.8498 0.0138 -1.3346>, rotation=<rot 70 -179.9999 -180>}]
 dyn_longhair_04 [{shapeType=Capsule, length=3.246, radius=3.2436, shapeName=dyn_longhair_04_capsule, center=<2.3171 0 0.6365>, rotation=<rot 0 0 90>}]
 dyn_sidehair_01_l [{shapeType=Capsule, length=3.35, radius=0.75, center=<1.4458 0 0>, rotation=<rot 90 0 180>}]
 dyn_sidehair_01_l_MSTR [{shapeType=Capsule, length=0.01, radius=0.01}]
 dyn_sidehair_01_r [{shapeType=Capsule, length=3.225, radius=0.9334, center=<-1.4458 0 0>, rotation=<rot 90 0 0>}]
 dyn_sidehair_01_r_MSTR [{shapeType=Capsule, length=0.01, radius=0.01}]
 dyn_sidehair_02_l [{shapeType=Capsule, length=3, radius=1.1, center=<1.4458 0 0>, rotation=<rot 90 0 180>}]
 dyn_sidehair_02_r [{shapeType=Capsule, length=3.225, radius=0.9334, center=<-1.4458 0 0>, rotation=<rot 90 0 0>}]
 dyn_tophair_01 [{shapeType=Box, extentX=4, extentY=5, extentZ=1.1, center=<1.5998 0 0>}]
 dyn_tophair_01_MSTR [{shapeType=Capsule, length=0.01, radius=0.01}]
 dyn_tophair_01_l [{shapeType=Box, extentX=4, extentY=5, extentZ=1.1, center=<1.5998 0 0>}]
 dyn_tophair_01_l_MSTR [{shapeType=Capsule, length=0.01, radius=0.01}]
 dyn_tophair_01_r [{shapeType=Box, extentX=4, extentY=5, extentZ=1.1, center=<-1.5998 0 0>, rotation=<rot 0 0 180>}]
 dyn_tophair_01_r_MSTR [{shapeType=Capsule, length=0.01, radius=0.01}]
 head [{shapeType=Capsule, length=5.4027, radius=7.3713, center=<3.6185 0.0974 0>, rotation=<rot -90 0 0.0001>}]
 neck_01 [{shapeType=Capsule, length=2.1527, radius=6.3713, center=<3.6185 0.0974 0>, rotation=<rot -90 0 0.0001>}]
 neck_02 [{shapeType=Capsule, length=2.4027, radius=4.8713, center=<3.6185 0.0974 0>, rotation=<rot -90 0 0.0001>}]
 spine_05 [
  {shapeType=Capsule, length=7.9027, radius=9.3713, center=<2.9134 1.2053 -5.0426>, rotation=<rot 30.0014 -179.9999 -179.9999>}
  {shapeType=Capsule, length=11.5, radius=9, center=<2.9134 0.8018 5.0426>, rotation=<rot -30.0014 -179.9999 -179.9999>}
  {shapeType=Capsule, length=10.5, radius=5, center=<-7.0463 -2.7421 0>}
 ]
 upperarm_l [{shapeType=Capsule, length=14, radius=5.005, center=<6.4975 0.0003 0.0164>, rotation=<rot -90 -0.0099 0.0099>}]
 upperarm_r [{shapeType=Capsule, length=14, radius=5.005, center=<-6.4975 -0.0003 -0.0164>, rotation=<rot -90 0.0099 179.9901>}]
}
constraints [
 {
  bone1Name=head
  bone2Name=neck_02
  swing1LimitDegrees=45
  swing1Motion=Free
  swing2LimitDegrees=45
  swing2Motion=Free
  twistLimitDegrees=45
  twistMotion=Free
 }
 {
  bone1Name=neck_01
  bone2Name=spine_05
  swing1LimitDegrees=45
  swing1Motion=Free
  swing2LimitDegrees=45
  swing2Motion=Free
  twistLimitDegrees=45
  twistMotion=Free
 }
 {
  bone1Name=neck_02
  bone2Name=neck_01
  swing1LimitDegrees=45
  swing1Motion=Free
  swing2LimitDegrees=45
  swing2Motion=Free
  twistLimitDegrees=45
  twistMotion=Free
 }
 {
  bone1Name=dyn_tophair_01_MSTR
  bone2Name=head
  swing1LimitDegrees=45
  swing1Motion=Free
  swing2LimitDegrees=45
  swing2Motion=Free
  twistLimitDegrees=45
  twistMotion=Free
 }
 {
  bone1Name=dyn_tophair_01_l_MSTR
  bone2Name=head
  swing1LimitDegrees=45
  swing1Motion=Free
  swing2LimitDegrees=45
  swing2Motion=Free
  twistLimitDegrees=45
  twistMotion=Free
 }
 {
  bone1Name=dyn_tophair_01_r_MSTR
  bone2Name=head
  swing1LimitDegrees=45
  swing1Motion=Free
  swing2LimitDegrees=45
  swing2Motion=Free
  twistLimitDegrees=45
  twistMotion=Free
 }
 {
  bone1Name=dyn_fronthair_01_l_MSTR
  bone2Name=head
  swing1LimitDegrees=45
  swing1Motion=Free
  swing2LimitDegrees=45
  swing2Motion=Free
  twistLimitDegrees=45
  twistMotion=Free
 }
 {
  bone1Name=dyn_sidehair_01_l_MSTR
  bone2Name=head
  swing1LimitDegrees=45
  swing1Motion=Free
  swing2LimitDegrees=45
  swing2Motion=Free
  twistLimitDegrees=45
  twistMotion=Free
 }
 {
  bone1Name=dyn_longhair_01_r_MSTR
  bone2Name=head
  swing1LimitDegrees=45
  swing1Motion=Free
  swing2LimitDegrees=45
  swing2Motion=Free
  twistLimitDegrees=45
  twistMotion=Free
 }
 {
  bone1Name=dyn_longhair_01_l_MSTR
  bone2Name=head
  swing1LimitDegrees=45
  swing1Motion=Free
  swing2LimitDegrees=45
  swing2Motion=Free
  twistLimitDegrees=45
  twistMotion=Free
 }
 {
  bone1Name=dyn_longhair_01_MSTR
  bone2Name=head
  swing1LimitDegrees=45
  swing1Motion=Free
  swing2LimitDegrees=45
  swing2Motion=Free
  twistLimitDegrees=45
  twistMotion=Free
 }
 {
  bone1Name=dyn_sidehair_01_r_MSTR
  bone2Name=head
  swing1LimitDegrees=45
  swing1Motion=Free
  swing2LimitDegrees=45
  swing2Motion=Free
  twistLimitDegrees=45
  twistMotion=Free
 }
 {
  bone1Name=dyn_fronthair_01_r_MSTR
  bone2Name=head
  swing1LimitDegrees=45
  swing1Motion=Free
  swing2LimitDegrees=45
  swing2Motion=Free
  twistLimitDegrees=45
  twistMotion=Free
 }
 {bone1Name=dyn_tophair_01, bone2Name=dyn_tophair_01_MSTR, swing1LimitDegrees=5, swing2LimitDegrees=15, twistLimitDegrees=5}
 {bone1Name=dyn_tophair_01_l, bone2Name=dyn_tophair_01_l_MSTR, swing1LimitDegrees=5, swing2LimitDegrees=15, twistLimitDegrees=5}
 {bone1Name=dyn_fronthair_01_l, bone2Name=dyn_fronthair_01_l_MSTR, swing1LimitDegrees=15, swing2LimitDegrees=45, twistLimitDegrees=5}
 {bone1Name=dyn_fronthair_02_l, bone2Name=dyn_fronthair_01_l, swing1LimitDegrees=15, swing2LimitDegrees=45, twistLimitDegrees=5}
 {bone1Name=dyn_longhair_01_l, bone2Name=dyn_longhair_01_l_MSTR, swing1LimitDegrees=15, swing2LimitDegrees=20, twistLimitDegrees=5}
 {bone1Name=dyn_longhair_02_l, bone2Name=dyn_longhair_01_l, swing1LimitDegrees=45, swing2LimitDegrees=45, twistLimitDegrees=5}
 {bone1Name=dyn_longhair_01, bone2Name=dyn_longhair_01_MSTR, swing1LimitDegrees=5, swing2LimitDegrees=25, twistLimitDegrees=5}
 {bone1Name=dyn_longhair_02, bone2Name=dyn_longhair_01, swing1LimitDegrees=5, swing2LimitDegrees=25, twistLimitDegrees=5}
 {bone1Name=dyn_longhair_03, bone2Name=dyn_longhair_02, swing1LimitDegrees=5, swing2LimitDegrees=25, twistLimitDegrees=5}
 {bone1Name=dyn_sidehair_01_r, bone2Name=dyn_sidehair_01_r_MSTR, swing1LimitDegrees=60, swing2LimitDegrees=60, twistLimitDegrees=15}
 {bone1Name=dyn_tophair_01_r, bone2Name=dyn_tophair_01_r_MSTR, swing1LimitDegrees=5, swing2LimitDegrees=15, twistLimitDegrees=5}
 {bone1Name=dyn_fronthair_02_r, bone2Name=dyn_fronthair_01_r, swing1LimitDegrees=45, swing2LimitDegrees=45, twistLimitDegrees=5}
 {bone1Name=dyn_fronthair_01_r, bone2Name=dyn_fronthair_01_r_MSTR, swing1LimitDegrees=45, swing2LimitDegrees=45, twistLimitDegrees=5}
 {bone1Name=dyn_sidehair_02_r, bone2Name=dyn_sidehair_01_r, swing1LimitDegrees=45, swing2LimitDegrees=45, twistLimitDegrees=5}
 {
  bone1Name=upperarm_l
  bone2Name=spine_05
  swing1LimitDegrees=45
  swing1Motion=Free
  swing2LimitDegrees=45
  swing2Motion=Free
  twistLimitDegrees=45
  twistMotion=Free
 }
 {
  bone1Name=upperarm_r
  bone2Name=spine_05
  swing1LimitDegrees=45
  swing1Motion=Free
  swing2LimitDegrees=45
  swing2Motion=Free
  twistLimitDegrees=45
  twistMotion=Free
 }
 {bone1Name=dyn_longhair_04, bone2Name=dyn_longhair_03, swing1LimitDegrees=5, swing2LimitDegrees=25, twistLimitDegrees=5}
 {bone1Name=dyn_longhair_03_l, bone2Name=dyn_longhair_02_l, swing1LimitDegrees=45, swing2LimitDegrees=45, twistLimitDegrees=5}
 {bone1Name=dyn_longhair_01_r, bone2Name=dyn_longhair_01_r_MSTR, swing1LimitDegrees=5, swing2LimitDegrees=20, twistLimitDegrees=5}
 {bone1Name=dyn_longhair_02_r, bone2Name=dyn_longhair_01_r, swing1LimitDegrees=5, swing2LimitDegrees=20, twistLimitDegrees=5}
 {bone1Name=dyn_longhair_03_r, bone2Name=dyn_longhair_02_r, swing1LimitDegrees=5, swing2LimitDegrees=45, twistLimitDegrees=5}
 {bone1Name=dyn_sidehair_01_l, bone2Name=dyn_sidehair_01_l_MSTR, swing1LimitDegrees=45, swing2LimitDegrees=45, twistLimitDegrees=5}
 {bone1Name=dyn_sidehair_02_l, bone2Name=dyn_sidehair_01_l, swing1LimitDegrees=45, swing2LimitDegrees=45, twistLimitDegrees=5}
]
