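# Post-process style AnimBP for the Nosh Hunter medium female face accessory mesh, adding secondary motion to hair and adjusting glasses over the pose copied from the attached parent.
# Clavicle constraints and Look Ats orient hair roots, and head swing drives a corrective before neck-based RigidBody simulation.
# Skydiving force or default gravity feeds simulation; bone offsets and blended mid-hair constraints follow.
# Crouch-aim, arm-swing and headphone trail nodes sit outside the output chain.
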AnimBlueprint /BRCosmetics/Characters/Player/Female/Medium/Bodies/F_MED_NoshHunter/Meshes/Parts/F_MED_Nosh_Hunter_FaceAcc_AnimBP
target_skeleton=&/F_MED_Nosh_Hunter_FaceAcc_Skeleton
graphs [
 AnimGraph {nodes [
   n1 = Get isCrouchingAndTargetingAlpha() -> (isCrouchingAndTargetingAlpha:double)
   n2 = "Constraint - dyn_longhair_01_r (R)"(Alpha=@n1, ConstraintWeights_0=1) {
    alphaBoolBlend {blendOption=Linear}
    boneToModify {boneName=dyn_longhair_01_r}
    constraintSetup [{transformType=Rotation, targetBone {boneName=dyn_longhair_02}}]
    constraintWeights [1]
   }
   n3 = "Constraint - dyn_longhair_02_r (R)"(ComponentPose=@n2, Alpha=@n1, ConstraintWeights_0=1) {
    alphaBoolBlend {blendOption=Linear}
    boneToModify {boneName=dyn_longhair_02_r}
    constraintSetup [{transformType=Rotation, targetBone {boneName=dyn_longhair_03}}]
    constraintWeights [1]
   }
   "Constraint - dyn_longhair_03_r (R)"(ComponentPose=@n3, Alpha=@n1, ConstraintWeights_0=1) {
    alphaBoolBlend {blendOption=Linear}
    boneToModify {boneName=dyn_longhair_03_r}
    constraintSetup [{transformType=Rotation, targetBone {boneName=dyn_longhair_04}}]
    constraintWeights [1]
   }
   "Transform (Modify) Bone - Bone: dyn_longhair_03_r"(Alpha=0, Translation="0,0,0", Rotation="5,0,0", Scale="0.85,0.85,0.85") {
    rotationMode=BMM_Additive
    rotationSpace=BCS_BoneSpace
    scaleMode=BMM_Additive
    scaleSpace=BCS_BoneSpace
    alphaBoolBlend {blendOption=Linear}
    boneToModify {boneName=dyn_longhair_03_r}
    rotation {pitch=5}
    scale=<0.85 0.85 0.85>
   }
   n4 = Get headSwingAngleAlpha() -> (headSwingAngleAlpha:double)
   "Transform (Modify) Bone - Bone: dyn_longhair_01_l_MSTR"(Alpha=@n4, Translation="0,0,0", Rotation="45,0,0", Scale="1,1,1") {
    rotationMode=BMM_Additive
    rotationSpace=BCS_BoneSpace
    translationMode=BMM_Additive
    translationSpace=BCS_BoneSpace
    alphaBoolBlend {blendOption=Linear}
    boneToModify {boneName=dyn_longhair_01_l_MSTR}
    rotation {pitch=45}
   }
   "Transform (Modify) Bone - Bone: dyn_longhair_01_r_MSTR"(Alpha=Get armSwingDelta_r() -> (armSwingDelta_r:double), Translation="4,0,-4", Rotation="0,0,0", Scale="1,1,1") {
    rotationMode=BMM_Additive
    rotationSpace=BCS_BoneSpace
    scaleMode=BMM_Additive
    scaleSpace=BCS_BoneSpace
    translationMode=BMM_Additive
    translationSpace=BCS_BoneSpace
    alphaBoolBlend {blendOption=Linear}
    boneToModify {boneName=dyn_longhair_01_r_MSTR}
    translation {x=4, z=-4}
   }
   "Transform (Modify) Bone - Bone: dyn_longhair_03"(Alpha=@n1, Translation="0,0,0", Rotation="0,0,0", Scale="0.8,1,1") {
    rotationMode=BMM_Additive
    rotationSpace=BCS_BoneSpace
    scaleMode=BMM_Additive
    scaleSpace=BCS_BoneSpace
    translationMode=BMM_Additive
    translationSpace=BCS_BoneSpace
    alphaBoolBlend {blendOption=Linear}
    boneToModify {boneName=dyn_longhair_03}
    scale {x=0.8}
   }
   n5 = "Constraint - dyn_sidehair_01_r_AIM (P)"(ComponentPose=Local To Component(LocalPose=Copy Pose From Mesh() {bUseAttachedParent=true}), Alpha=1, ConstraintWeights_0=1) {
    alphaBoolBlend {blendOption=Linear}
    boneToModify {boneName=dyn_sidehair_01_r_AIM}
    constraintSetup [{targetBone {boneName=clavicle_r}}]
    constraintWeights [1]
   }
   n6 = "Constraint - dyn_longhair_01_r_AIM (P)"(ComponentPose=@n5, Alpha=1, ConstraintWeights_0=1) {
    alphaBoolBlend {blendOption=Linear}
    boneToModify {boneName=dyn_longhair_01_r_AIM}
    constraintSetup [{targetBone {boneName=clavicle_r}}]
    constraintWeights [1]
   }
   n7 = "Constraint - dyn_fronthair_01_r_AIM (P)"(ComponentPose=@n6, Alpha=1, ConstraintWeights_0=1) {
    alphaBoolBlend {blendOption=Linear}
    boneToModify {boneName=dyn_fronthair_01_r_AIM}
    constraintSetup [{targetBone {boneName=clavicle_r}}]
    constraintWeights [1]
   }
   n8 = Look At - Bone: dyn_longhair_01_MSTR(ComponentPose=@n7, Alpha=1) {
    alphaBoolBlend {blendOption=Linear}
    boneToModify {boneName=dyn_longhair_01_MSTR}
    lookAtLocation {x=0, z=3}
    lookAtTarget {boneReference {boneName=dyn_longhair_01_AIM}}
    lookAt_Axis {axis=<1 0>}
    lookUp_Axis {axis {x=0, z=1}}
   }
   n9 = Look At - Bone: dyn_longhair_01_l_MSTR(ComponentPose=@n8, Alpha=1) {
    alphaBoolBlend {blendOption=Linear}
    boneToModify {boneName=dyn_longhair_01_l_MSTR}
    lookAtLocation {x=0, z=3}
    lookAtTarget {boneReference {boneName=dyn_longhair_01_l_AIM}}
    lookAt_Axis {axis=<1 0>}
    lookUp_Axis {axis {x=0, z=1}}
   }
   n10 = Look At - Bone: dyn_sidehair_01_l_MSTR(ComponentPose=@n9, Alpha=1) {
    alphaBoolBlend {blendOption=Linear}
    boneToModify {boneName=dyn_sidehair_01_l_MSTR}
    lookAtLocation {x=0, z=1}
    lookAtTarget {boneReference {boneName=dyn_sidehair_01_l_AIM}}
    lookAt_Axis {axis=<1 0>}
    lookUp_Axis {axis {x=0, z=1}}
   }
   n11 = Look At - Bone: dyn_fronthair_01_l_MSTR(ComponentPose=@n10, Alpha=1) {
    alphaBoolBlend {blendOption=Linear}
    boneToModify {boneName=dyn_fronthair_01_l_MSTR}
    lookAtLocation {x=0, z=1}
    lookAtTarget {boneReference {boneName=dyn_fronthair_01_l_AIM}}
    lookAt_Axis {axis=<1 0>}
    lookUp_Axis {axis {x=0, z=1}}
   }
   n12 = Look At - Bone: dyn_longhair_01_r_MSTR(ComponentPose=@n11, Alpha=1) {
    alphaBoolBlend {blendOption=Linear}
    boneToModify {boneName=dyn_longhair_01_r_MSTR}
    lookAtLocation {x=0, z=-3}
    lookAtTarget {boneReference {boneName=dyn_longhair_01_r_AIM}}
    lookAt_Axis {axis=<-1 0>}
    lookUp_Axis {axis {x=0, z=-1}}
   }
   n13 = Look At - Bone: dyn_sidehair_01_r_MSTR(ComponentPose=@n12, Alpha=1) {
    alphaBoolBlend {blendOption=Linear}
    boneToModify {boneName=dyn_sidehair_01_r_MSTR}
    lookAtLocation {x=0, z=-1}
    lookAtTarget {boneReference {boneName=dyn_sidehair_01_r_AIM}}
    lookAt_Axis {axis=<-1 0>}
    lookUp_Axis {axis {x=0, z=-1}}
   }
   n14 = Look At - Bone: dyn_fronthair_01_r_MSTR(ComponentPose=@n13, Alpha=1) {
    alphaBoolBlend {blendOption=Linear}
    boneToModify {boneName=dyn_fronthair_01_r_MSTR}
    lookAtLocation {x=0, z=-1}
    lookAtTarget {boneReference {boneName=dyn_fronthair_01_r_AIM}}
    lookAt_Axis {axis=<-1 0>}
    lookUp_Axis {axis {x=0, z=-1}}
   }
   n15 = "Transform (Modify) Bone - Bone: dyn_longhair_01_MSTR"(ComponentPose=@n14, Alpha=@n4, Translation="0,0.000,-2", Rotation="30,0,0", Scale="1,1,1") {
    rotationMode=BMM_Additive
    rotationSpace=BCS_BoneSpace
    translationMode=BMM_Additive
    translationSpace=BCS_BoneSpace
    alphaBoolBlend {blendOption=Linear}
    boneToModify {boneName=dyn_longhair_01_MSTR}
    rotation {pitch=30}
    translation {z=-2}
   }
   n16 = "Transform (Modify) Bone - Bone: dyn_longhair_01_r_MSTR"(ComponentPose=@n15, Alpha=1, Translation="2,0.000,-2", Rotation="45,0,0", Scale="1,1,1") {
    rotationMode=BMM_Additive
    rotationSpace=BCS_BoneSpace
    translationMode=BMM_Additive
    translationSpace=BCS_BoneSpace
    alphaBoolBlend {blendOption=Linear}
    boneToModify {boneName=dyn_longhair_01_r_MSTR}
    rotation {pitch=45}
    translation {x=2, z=-2}
   }
   n17 = RigidBody!(ComponentPose=@n16, Alpha, OverrideWorldGravity:Vector=Get windVector() -> (windVector:Vector), ExternalForce) -> (Pose) {
    bClampLinearTranslationLimitToRefPose=true
    bForceDisableCollisionBetweenConstraintBodies=true
    bOverrideWorldGravity=true
    overridePhysicsAsset=&/F_MED_Nosh_Hunter_FaceAcc_Physics
    simulationSpace=BaseBoneSpace
    alphaBoolBlend {blendOption=Linear}
    baseBoneRef {boneName=neck_01}
    componentAppliedLinearAccClamp=<3000 3000 3000>
    componentLinearAccScale=<0.3 0.3 0.6>
    componentLinearVelScale=<0.7 0.7 0.6>
    overrideWorldGravity {z=-980}
   }
   n18 = "Transform (Modify) Bone - Bone: dyn_longhair_01_MSTR"(ComponentPose=@n17, Alpha=1, Translation="0,-1,0", Rotation="6,0,0", Scale="1,1,1") {
    rotationMode=BMM_Additive
    rotationSpace=BCS_BoneSpace
    translationMode=BMM_Additive
    alphaBoolBlend {blendOption=Linear}
    boneToModify {boneName=dyn_longhair_01_MSTR}
    rotation {pitch=6}
    translation {y=-1}
   }
   n19 = "Transform (Modify) Bone - Bone: dyn_longhair_01_l_MSTR"(ComponentPose=@n18, Alpha=1, Translation="0,1,0", Rotation="0,0,0", Scale="1,1,1") {
    rotationMode=BMM_Additive
    rotationSpace=BCS_BoneSpace
    translationMode=BMM_Additive
    alphaBoolBlend {blendOption=Linear}
    boneToModify {boneName=dyn_longhair_01_l_MSTR}
    translation {y=1}
   }
   n20 = "Transform (Modify) Bone - Bone: dyn_longhair_01_r_MSTR"(ComponentPose=@n19, Alpha=1, Translation="0,1,-1", Rotation="0,0,0", Scale="1,1,1") {
    rotationMode=BMM_Additive
    rotationSpace=BCS_BoneSpace
    translationMode=BMM_Additive
    alphaBoolBlend {blendOption=Linear}
    boneToModify {boneName=dyn_longhair_01_r_MSTR}
    translation {y=1, z=-1}
   }
   n21 = "Transform (Modify) Bone - Bone: dyn_longhair_03"(ComponentPose=@n20, Alpha=1, Translation="0,2,2", Rotation="5,0,0", Scale="1,1,1") {
    rotationMode=BMM_Additive
    rotationSpace=BCS_BoneSpace
    translationMode=BMM_Additive
    alphaBoolBlend {blendOption=Linear}
    boneToModify {boneName=dyn_longhair_03}
    rotation {pitch=5}
    translation {y=2, z=2}
   }
   n22 = "Transform (Modify) Bone - Bone: dyn_longhair_01_l_MSTR"(ComponentPose=@n21, Alpha=1, Translation="0,0,0", Rotation="0,0,0", Scale="1,1,1") {
    rotationMode=BMM_Additive
    rotationSpace=BCS_BoneSpace
    translationMode=BMM_Additive
    alphaBoolBlend {blendOption=Linear}
    boneToModify {boneName=dyn_longhair_01_l_MSTR}
   }
   n23 = "Transform (Modify) Bone - Bone: dyn_longhair_01_r_MSTR"(ComponentPose=@n22, Alpha=1, Translation="0,0,1", Rotation="0,0,0", Scale="1,1,1") {
    rotationMode=BMM_Additive
    rotationSpace=BCS_BoneSpace
    translationMode=BMM_Additive
    alphaBoolBlend {blendOption=Linear}
    boneToModify {boneName=dyn_longhair_01_r_MSTR}
    translation {z=1}
   }
   n24 = "Transform (Modify) Bone - Bone: dyn_longhair_03"(ComponentPose=@n23, Alpha=1, Translation="0,3,0", Rotation="0,0,0", Scale="1,1,1") {
    rotationMode=BMM_Additive
    rotationSpace=BCS_BoneSpace
    translationMode=BMM_Additive
    alphaBoolBlend {blendOption=Linear}
    boneToModify {boneName=dyn_longhair_03}
    translation {y=3}
   }
   n25 = "Transform (Modify) Bone - Bone: dyn_sidehair_01_l_MSTR"(ComponentPose=@n24, Alpha=1, Translation="1,0,0", Rotation="7,0,0", Scale="1,1,1") {
    rotationMode=BMM_Additive
    rotationSpace=BCS_BoneSpace
    translationMode=BMM_Additive
    alphaBoolBlend {blendOption=Linear}
    boneToModify {boneName=dyn_sidehair_01_l_MSTR}
    rotation {pitch=7}
    translation {x=1}
   }
   n26 = "Transform (Modify) Bone - Bone: dyn_fronthair_01_r_MSTR"(ComponentPose=@n25, Alpha=1, Translation="-1,0,0.5", Rotation="5,0,0", Scale="1,1,1") {
    rotationMode=BMM_Additive
    rotationSpace=BCS_BoneSpace
    translationMode=BMM_Additive
    alphaBoolBlend {blendOption=Linear}
    boneToModify {boneName=dyn_fronthair_01_r_MSTR}
    rotation {pitch=5}
    translation {x=-1, z=0.5}
   }
   n27 = "Constraint - dyn_longhair_mid_01_r (PP)"!(ComponentPose=@n26, Alpha=1, ConstraintWeights_0=0.75, ConstraintWeights_1:float=0.25) -> (Pose) {
    alphaBoolBlend {blendOption=Linear}
    boneToModify {boneName=dyn_longhair_mid_01_r}
    constraintSetup [{targetBone {boneName=dyn_longhair_02}}, {targetBone {boneName=dyn_longhair_01_r}}]
    constraintWeights [1,1]
   }
   n28 = "Constraint - dyn_longhair_mid_02_r (PP)"!(ComponentPose=@n27, Alpha=1, ConstraintWeights_0=0.75, ConstraintWeights_1:float=0.25) -> (Pose) {
    alphaBoolBlend {blendOption=Linear}
    boneToModify {boneName=dyn_longhair_mid_02_r}
    constraintSetup [{targetBone {boneName=dyn_longhair_03}}, {targetBone {boneName=dyn_longhair_02_r}}]
    constraintWeights [1,1]
   }
   n29 = "Constraint - dyn_longhair_mid_03_r (PP)"!(ComponentPose=@n28, Alpha=1, ConstraintWeights_0=0.75, ConstraintWeights_1:float=0.25) -> (Pose) {
    alphaBoolBlend {blendOption=Linear}
    boneToModify {boneName=dyn_longhair_mid_03_r}
    constraintSetup [{targetBone {boneName=dyn_longhair_04}}, {targetBone {boneName=dyn_longhair_03_r}}]
    constraintWeights [1,1]
   }
   n30 = "Transform (Modify) Bone - Bone: glasses_01_DSP"(ComponentPose=@n29, Alpha=1, Translation="-0.000,0.431,0.000", Rotation="0,0,0", Scale="1,1,1") {
    rotationMode=BMM_Additive
    rotationSpace=BCS_BoneSpace
    translationMode=BMM_Additive
    translationSpace=BCS_BoneSpace
    alphaBoolBlend {blendOption=Linear}
    boneToModify {boneName=glasses_01_DSP}
    translation {y=0.4315}
   }
   Output Pose(Result=Component To Local(ComponentPose=@n30)) {layerGroup=None}
   Trail controller - Bone: dyn_headphone_02_r!(ComponentPose, Alpha=0.1, RelaxationSpeedScale:float=4) -> (Pose) {
    maxDeltaTime=0.0333
    relaxationSpeedScale=3
    rotationLimits [{limitMax=<5 15 65>, limitMin=<-5 -15 -65>}, {limitMax=<5 15 65>, limitMin=<-5 -15 -65>}]
    rotationOffsets [<0 0 0>, <0 0 0>]
    trailBone {boneName=dyn_headphone_02_r}
   }
  ]}
 EventGraph {nodes [
   ResetDynamics(InTeleportType=TeleportPhysics)
   > ResetDynamics(InTeleportType=ResetPhysics)
   ResetDynamics(InTeleportType=TeleportPhysics)
   > ResetDynamics(InTeleportType=ResetPhysics)
   Delay(Duration=0.125)
   > ResetDynamics(InTeleportType=TeleportPhysics)
   > ResetDynamics(InTeleportType=ResetPhysics)
   Event BlueprintUpdateAnimation()
   > n1 = Sequence()
   n2 = Break Fort Anim Input Skydiving External Force(FortAnimInput_SkydivingExternalForce=Get SkydivingExternalForceInput()) -> (
    bUseSkydivingVectorForce=~
    SkydivingVectorForce=~
    HeadToPelvisDirection=~
    FloatingMultiplier=~
    FloatingAdditive=~
    DivingMultiplier=~
    DivingAdditive=~
    ParachutingMultiplier=~
    ParachutingAdditive=~
   )
   > Set windVector(windVector:Vector=SelectVector(A=vector * float(A=@n2.SkydivingVectorForce, B=0.15), B="0, 0,-980", bPickA=Get bIsSkyDiving())) -> (Output_Get:Vector)
   n3 = GetOwnerSkeletalMeshComponent()
   n4 = GetDeltaTransformFromRefPose(self=@n3, BoneName=upperarm_r, BaseName=clavicle_r)
   n5 = "Absolute (Float)"(A=BreakRotator(InRot=BreakTransform(InTransform=@n4).Rotation).Yaw)
   Set armSwingDelta_r(
    execute=@n1.then_1
    armSwingDelta_r:double=SelectFloat(A=MapRangeClamped(Value=@n5, InRangeA=5, InRangeB=70), B=0, bPickA=Get bIsTargeting())
   ) -> (Output_Get:double)
   > Set headSwingAngleAlpha(headSwingAngleAlpha:double=MapRangeClamped(Value=GetTwistAndSwingAngleOfDeltaRotationFromRefPose(self=@n3, BoneName=head).OutSwingAngle, InRangeA=10, InRangeB=30)) -> (Output_Get:double)
   n6 = Get bIsCrouching()
   n7 = Get bIsTargeting()
   n8 = Get bIsCrouching()
   > Set isCrouchingAndTargetingAlpha(isCrouchingAndTargetingAlpha:double=SelectFloat(A=1, B=0, bPickA=AND Boolean!(A=@n7, B=@n6, C:bool=@n8) -> (ReturnValue))) -> (Output_Get:double)
   TryGetPawnOwner()
   Event BlueprintInitializeAnimation()
   Event OnBeginCrouching()
   Event OnEndCrouching()
  ]}
]
variables [
 armSwingDelta_r {type=number}
 headSwingAngleAlpha {type=number}
 isCrouchingAndTargetingAlpha {type=number}
 windVector {type=Vector}
]
